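# Niagara landing effect for the M_MED_Drift_Winter cosmetic body, shedding frost, ice and snow from the User.SkeletalMesh.
# Frost is a short CPU burst of SubUV-animated sprites, more opaque when User.FrontendSpawn is set.
# GPU emitters Ice and SnowFall taper spawn probability over their loop, inherit capped horizontal velocity, then fall as tumbling bits or curl-noise flakes.
# Both GPU emitters scale sprite size by a local camera-fade module.
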NiagaraSystem /BRCosmetics/Characters/Player/Male/Medium/Bodies/M_MED_Drift_Winter/Effects/NS_DriftWinter_Landing
dependencies {
 usedDataInterfaces [&NiagaraDataInterfaceCamera, &NiagaraDataInterfaceCurve]
 usedDynamicInputs []
 usedModules [&SystemState]
}
emitters {
 Frost {
  emitterUpdateScript [
   &EmitterState(Life Cycle Mode=::NewEnumerator1, Loop Behavior=::NewEnumerator1, Loop Delay=0.01, Loop Duration=0.3, UseLoopDelay=true)
   &SpawnBurst_Instantaneous(Spawn Count=30, Use Spawn Probability=false)
  ]
  particleSpawnScript [
   /Niagara/Modules/Spawn/Initialization/InitializeParticle(
    Color=&MakeCustomLinearColorFromBool(Bool=$User.FrontendSpawn, True Color=<rgba 0.2779 0.6803 1 0.7>, False Color=<rgba 0.2779 0.6803 1 0.2>)
    Lifetime=&RandomRangeFloat(Minimum=0.4, Maximum=0.7)
    Module.Write Color=true
    Module.Write Lifetime=true
    Module.Write Mass=true
    Module.Write Material Random=true
    Module.Write Position=true
    Module.Write Scale=false
    Module.Write SpriteRotation=true
    Module.Write SpriteSize=true
    Sprite Rotation=&RandomRangeFloat(Minimum=-180, Maximum=180)
    Sprite Size=&Vector2DFromFloat(Value=&RandomRangeFloat(Minimum=18, Maximum=22))
   )
   &SkeletalMeshLocation(
    Inherit Velocity=false
    MaskSpawnGroup=false
    Mesh Sampling Type=::NewEnumerator7
    Module.Override Randomness=false
    OffsetAlongNormal=false
    Skeletal Mesh=$User.SkeletalMesh
   )
   &AddVelocity(Coordinate Space=::World, Velocity=&RandomRangeVector(Minimum=<0.1 0.1 -30>, Maximum=<-0.1 -0.1 -70>))
   &CameraOffset(Camera Offset Amount=30)
  ]
  particleUpdateScript [
   &ParticleState()
   &SubUVAnimation(End Frame=40, Start Frame=2)
   &GravityForce(Gravity=&RandomRangeVector(Minimum=<0 0 -6>, Maximum=<0 0 -20>))
   &SolveForcesAndVelocity(Clamp Velocity=false, Limit Acceleration=false)
   &ScaleColor(
    Scale Alpha=&FloatFromCurve(FloatCurve=DataInterface() {propertyValues {Curve {keys [
         {arriveTangent=0.1402, interpMode=RCIM_Cubic, leaveTangent=0.1402, tangentMode=RCTM_User, time=0, value=0}
         {arriveTangent=0.317, interpMode=RCIM_Cubic, leaveTangent=0.317, tangentMode=RCTM_User, time=0.4, value=0.4}
         <at 0.7886 0 RCIM_Cubic>
        ]}}})
    ScaleA=true
    ScaleRGB=true
   )
   &DynamicMaterialParameters(
    Index 0 Param 4=1
    Module.Param0WriteEnabled=true
    Module.Param1WriteEnabled=true
    Module.Param2WriteEnabled=true
    Module.Param3WriteEnabled=true
   )
   &ScaleSpriteSize(Scale Factor=&Vector2DFromFloat(Value=&FloatFromCurve(FloatCurve=DataInterface() {propertyValues {Curve {keys [<at 0 1 RCIM_Cubic>, {arriveTangent=0.8108, interpMode=RCIM_Cubic, leaveTangent=0.8108, tangentMode=RCTM_User, time=1, value=1.5}]}}})))
  ]
  rendererClasses [&NiagaraSpriteRendererProperties]
  renderers [{rendererClass=&NiagaraSpriteRendererProperties}]
 }
 Ice {
  simTarget=GPUComputeSim
  emitterUpdateScript [
   &EmitterState(Life Cycle Mode=::NewEnumerator1, Loop Behavior=::NewEnumerator1, Loop Duration=0.2, UseLoopDelay=false)
   &SpawnRate(Spawn Probability=&OneMinusFloat(Float=$Emitter.NormalizedLoopAge), SpawnRate=300, Use Spawn Probability=true)
  ]
  particleSpawnScript [
   &InitializeParticle(
    Color=&RandomRangeLinearColor(Minimum=<rgba 0.25 0.5375 1 1>, Maximum=<rgba 0.2 0.6 1.5 1>)
    Lifetime=&UniformRangedFloat(Minimum=0.5, Maximum=1.2)
    Module.Write Color=true
    Module.Write Lifetime=true
    Module.Write Mass=false
    Module.Write Material Random=true
    Module.Write Position=false
    Module.Write Scale=false
    Module.Write SpriteRotation=true
    Module.Write SpriteSize=true
    Position=&SimulationPosition()
    Sprite Rotation=&UniformRangedFloat(Minimum=-360, Maximum=360)
    Sprite Size=&Multiply_Vector2(
     A=&RandomRangeVector2D(Minimum=<2 5>, Maximum=<5 9>)
     B=&Vector2DFromFloat(Value=&RandomRangeFloat(Minimum=0.2, Maximum=1.1))
    )
   )
   &SkeletalMeshLocation(
    Inherit Velocity=false
    MaskSpawnGroup=false
    Mesh Sampling Type=::NewEnumerator7
    Module.Override Randomness=false
    OffsetAlongNormal=false
    Sampled Position Offset=<0 0 20>
    Skeletal Mesh=$User.SkeletalMesh
   )
   &CameraOffset(Camera Offset Amount=30)
   &AddVelocity(Coordinate Space=::World, Velocity=&RandomRangeVector(Minimum=<20 20 -10>, Maximum=<-20 -20 -70>))
   &InheritVelocity(
    Module.Limit Speed=true
    Velocity Limit:NiagaraFloat=&RandomRangeFloat(Minimum=100, Maximum=300)
    Velocity Scale=<1 1 0>
   )
  ]
  particleUpdateScript [
   &self:CameraFadeModule(FadeDistance=100, Offset=50, Particle Position:NiagaraPosition=$Particles.Position) {module=CameraFadeModule}
   &ParticleState()
   &GravityForce(Gravity=&RandomRangeVector(Minimum=<0 0 -100>, Maximum=<0 0 -800>))
   &SpriteRotationRate(Rotation Rate=&RandomRangeFloat(Minimum=-400, Maximum=600))
   &Drag(Use Linear Drag=true, Use Rotational Drag=false)
   &SolveForcesAndVelocity(Clamp Velocity=false, Limit Acceleration=false)
   &ScaleSpriteSize(Scale Factor=&Vector2DFromFloat(Value=&FloatFromCurve(
      FloatCurve=DataInterface() {propertyValues {Curve {keys [
          <at 0 0 RCIM_Cubic>
          <at 0.2 1 RCIM_Cubic>
          {arriveTangent=-1.5163, interpMode=RCIM_Cubic, leaveTangent=-1.5163, tangentMode=RCTM_User, time=1, value=0}
         ]}}}
      Scale Curve=$Particles.CameraFade
     )))
   &ScaleColor(
    Scale Alpha=&FloatFromCurve(FloatCurve=DataInterface() {propertyValues {Curve {keys [<at 0.8 1 RCIM_Cubic>,<at 1 0 RCIM_Cubic>]}}})
    ScaleA=true
    ScaleRGB=true
   )
  ]
  rendererClasses [&NiagaraSpriteRendererProperties]
  renderers [{rendererClass=&NiagaraSpriteRendererProperties}]
 }
 SnowFall {
  simTarget=GPUComputeSim
  emitterUpdateScript [
   &EmitterState(Life Cycle Mode=::NewEnumerator1, Loop Behavior=::NewEnumerator1, Loop Duration=0.5, UseLoopDelay=false)
   &SpawnRate(Spawn Probability=&OneMinusFloat(Float=$Emitter.NormalizedLoopAge), SpawnRate=90, Use Spawn Probability=true)
  ]
  particleSpawnScript [
   &InitializeParticle(
    Color=<rgba 0.5 0.75 1 1>
    Lifetime=&RandomRangeFloat(Minimum=1.8, Maximum=0.5)
    Module.Write Color=true
    Module.Write Lifetime=true
    Module.Write Mass=true
    Module.Write Material Random=true
    Module.Write Position=false
    Module.Write Scale=false
    Module.Write SpriteRotation=false
    Module.Write SpriteSize=true
    Position=&SimulationPosition()
    Sprite Size=&RandomRangeVector2D(Minimum=<1.5 1.5>, Maximum=<0.5 2>)
   )
   &SkeletalMeshLocation(
    Inherit Velocity=false
    MaskSpawnGroup=false
    Mesh Sampling Type=::NewEnumerator7
    Module.Override Randomness=false
    OffsetAlongNormal=false
    Sampled Position Offset=<0 0 5>
    Skeletal Mesh=$User.SkeletalMesh
   )
   &AddVelocity(Coordinate Space=::World, Velocity=&RandomRangeVector(Minimum=<0.1 0.1 -15>, Maximum=<-0.1 -0.1 -50>))
   &CameraOffset(Camera Offset Amount=30)
   &InheritVelocity(
    Module.Limit Speed=true
    Velocity Limit:NiagaraFloat=&RandomRangeFloat(Minimum=100, Maximum=300)
    Velocity Scale=<1 1 0>
   )
  ]
  particleUpdateScript [
   &self:CameraFadeModule(FadeDistance=100, Offset=50, Particle Position:NiagaraPosition=$Particles.Position) {module=CameraFadeModule}
   &ParticleState()
   &GravityForce(Gravity=&RandomRangeVector(Minimum=<0 0 -30>, Maximum=<0 0 -80>))
   &CurlNoiseForce(Noise Strength=50, Pan Noise=false)
   &SolveForcesAndVelocity(Clamp Velocity=false, Limit Acceleration=false)
   &ScaleSpriteSize(Scale Factor=&Vector2DFromFloat(Value=&FloatFromCurve(FloatCurve=DataInterface() {propertyValues {Curve {keys [<at 0.65 1>,<at 1 0>]}}}, Scale Curve=$Particles.CameraFade)))
  ]
  rendererClasses [&NiagaraSpriteRendererProperties]
  renderers [{rendererClass=&NiagaraSpriteRendererProperties}]
 }
}
summary {
 systemName=NS_DriftWinter_Landing
 emitters [
  {emitterName=Ice, simTarget=GPUComputeSim, rendererClasses [&NiagaraSpriteRendererProperties]}
  {emitterName=SnowFall, simTarget=GPUComputeSim, rendererClasses [&NiagaraSpriteRendererProperties]}
  {emitterName=Frost, rendererClasses [&NiagaraSpriteRendererProperties]}
 ]
 userVariables [
  User.FrontendSpawn {type=NiagaraBool}
  User.SkeletalMesh {
   type=NiagaraDataInterfaceSkeletalMesh
   defaultValue {
    struct=NiagaraExt_VariableValue_DataInterface
    value {dataInterface=&self:NiagaraDataInterfaceSkeletalMesh_0}
   }
  }
 ]
}
